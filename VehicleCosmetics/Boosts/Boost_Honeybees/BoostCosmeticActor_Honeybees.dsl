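; Honeybees boost cosmetic actor for vehicles, extending a parent boost class and driving the NS_Boost1 Niagara effect.
; Color-ready sets the ColorVariantID variant index, and OffsetPerCar optionally applies a per-body turbo offset from a data table.
; Single-boost setup centers SingleBoostPos using body and right-turbo socket positions, otherwise zeroing it.
; When boost finishes, it sets the effect active, then deactivates it after a one-second delay.
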(event EventBeginPlay
  (|CallParentFunction)
  (Niagara|SetNiagaraVariableByString(Vector3) (Variables|Default|GetNS_Boost1) "" "0, 0, 0"))

(event Collision|EventActorBeginOverlap
  (params (OtherActor :type Actor%20Object%20Reference)))

(event EventTick
  (params (DeltaSeconds :type Float%20%28single-precision%29)))

(event Vehicle|Cosmetics|EventOnBoostFinished
  (bind _ns_boost1 (Variables|Default|GetNS_Boost1))
  (|CallParentFunction)
  (Components|Activation|SetActive _ns_boost1 true false)
  (Utilities|FlowControl|Delay 1.0)
  (Components|Activation|SetActive _ns_boost1 false false))

(event Vehicle|Cosmetics|EventOnColorReady
  (params (Color :type Name :default "None") (AdditionalInfo :type Vehicle%20Cosmetics%20Additional%20Parameter%20Infos%20Structure))
  (|CallParentFunction Color AdditionalInfo)
  (bind _color_index (Class|BPVehicleCosmeticsFunctions|BPGetNSIndexbyColor Color))
  (Niagara|SetNiagaraVariableByString(Int32) (Variables|Default|GetNS_Boost1) "ColorVariantID" _color_index))

(fn OffsetPerCar
  (params (VehicleCosmeticsInterface :type Vehicle%20Cosmetics%20Interface%20Interface) (Boost :type Niagara%20Particle%20System%20Component%20Object%20Reference) (OffsetVariableName :type Name :default "None"))
  (|CallParentFunction VehicleCosmeticsInterface Boost OffsetVariableName)
  (if (and (Utilities|IsValidInterface VehicleCosmeticsInterface) (Variables|Default|GetUseOffsetPerCar))
    (bind _returnvalue (Utilities|GetDataTableRow "/VehicleCosmetics/DataTables/DT_VehicleCosmetics_TurboOffsets.DT_VehicleCosmetics_TurboOffsets" (Vehicle|Cosmetics|GetBodyItemName VehicleCosmeticsInterface)))
    (Niagara|SetNiagaraVariable(Vector3) (Variables|Default|GetNS_Boost1) OffsetVariableName (Utilities|Struct|BreakStructVehiclesCosmeticsTurboOffset _returnvalue))
    (return)))

(fn SingleBoostSetup
  (bind _returnvalue_location (Math|Transform|BreakTransform (Transformation|GetSocketTransform (Vehicle|Cosmetics|GetBodyMeshComponent) "Socket_Body" "RTS_ParentBoneSpace")))
  (bind (_returnvalue_location_x _returnvalue_location_y) (Math|Vector|BreakVector _returnvalue_location))
  (bind _returnvalue_location_2 (Math|Transform|BreakTransform (Transformation|GetSocketTransform (Vehicle|Cosmetics|GetBodyMeshComponent) "Socket_Turbo_A_R" "RTS_ParentBoneSpace")))
  (bind (_returnvalue_location_x_2 _returnvalue_location_y_2 _returnvalue_location_z) (Math|Vector|BreakVector _returnvalue_location_2))
  (bind _ns_boost1 (Variables|Default|GetNS_Boost1))
  (|CallParentFunction)
  (if (Variables|Default|GetSingleBoost?)
    (Niagara|SetNiagaraVariable(Vector3) _ns_boost1 "SingleBoostPos" (Math|Vector|MakeVector _returnvalue_location_x_2 _returnvalue_location_y _returnvalue_location_z))
    (else
      (Niagara|SetNiagaraVariable(Vector3) _ns_boost1 "SingleBoostPos" "0, 0, 0"))))

(fn ConstructionScript
  (|CallParentFunction))
