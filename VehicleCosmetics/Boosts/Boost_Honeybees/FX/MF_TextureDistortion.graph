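# Material function in the Honeybees boost cosmetic FX that turns a panning texture into a UV distortion offset.
# Tiled UVs scroll by Panner Time and Speed inputs, sample the input texture as a normal map, and a Channel Selector dot product isolates a channel, red by default.
# Distortion Strength scales that scalar, output alone or added equally to both UV axes.
# The "Mask After Strength Mul" output carries the unscaled value, despite its name.
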MaterialFunction /VehicleCosmetics/Boosts/Boost_Honeybees/FX/MF_TextureDistortion
graph {nodes [
  n1 = Multiply(A=TextureCoordinate() {uTiling=0.66, vTiling=0.66}, B=FunctionInput() {inputName=Tiling, inputType=FunctionInput_Vector2, sortPriority=1})
  n2 = Multiply(
   A=Time()
   B=FunctionInput(Preview=Constant() {r=1}) {bUsePreviewValueAsDefault=true, inputName=Panner Time, inputType=FunctionInput_Scalar, sortPriority=2}
  )
  n3 = FunctionInput() {bUsePreviewValueAsDefault=true, inputName=Panner Speed, inputType=FunctionInput_Vector2, sortPriority=3, previewValue=<1 1 0 1>}
  n4 = TextureObject()
  n5 = DotProduct(
   A=TextureSample(
    UVs=Panner(Coordinate=@n1, Time=@n2, Speed=@n3) {speedX=1, speedY=1}
    Tex=FunctionInput(Preview=@n4) {inputName=Texture, inputType=FunctionInput_Texture2D}
   ) {samplerType=SAMPLERTYPE_Normal}.RGB
   B=FunctionInput(Preview=Constant3Vector() {constant=<rgba 1 0 0 0>}) {bUsePreviewValueAsDefault=true, inputName=Channel Selector, sortPriority=5}
  )
  n6 = FunctionInput() {bUsePreviewValueAsDefault=true, inputName=Distortion Strength, inputType=FunctionInput_Scalar, sortPriority=4, previewValue=<0.05 0 0 1>}
  n7 = Multiply(A=@n5, B=@n6) {constB=0.05}
  FunctionOutput(None=@n7) {outputName=Distortion}
  FunctionOutput(None=@n5) {outputName=Mask After Strength Mul, sortPriority=1}
  FunctionOutput(None=Add(A=TextureCoordinate(), B=@n7)) {outputName=Distortion with UVs, sortPriority=2}
 ]}
