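# Niagara burst effect for the Honeybees vehicle boost cosmetic, showing a beehive exploding with honey behind the car.
# HoneyBeesHive briefly swells a hive, BrokeHive scatters fading fragments, and HoneyExplosion001 throws small droplets plus a few large SubUV splats.
# HoneyExplosion001 merges its two layers into one emitter, using spawn groups to select color, size and motion.
# All emitters apply painted color variants and share a local offset driven by car and boost-position user parameters.
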NiagaraSystem /VehicleCosmetics/Boosts/Boost_Honeybees/FX/NS_Boost_Honeybees_Explosion
dependencies {
 usedDataInterfaces [
  &NiagaraDataInterfaceMeshRendererInfo
  &NiagaraDataInterfaceCurve
  &NiagaraDataInterfaceSpriteRendererInfo
  &NiagaraDataInterfaceSkeletalMesh
 ]
 usedDynamicInputs []
 usedModules [
  &self:SystemSpawnScript.SystemScriptSource.SystemScriptGraph.NiagaraNodeAssignment_0.SetVariables_#1
  &SystemState
 ]
}
emitters {
 BrokeHive {
  emitterUpdateScript [
   &EmitterState(Life Cycle Mode=::NewEnumerator1, Loop Behavior=::NewEnumerator1, Loop Delay=0.1, UseLoopDelay=true)
   &SpawnRate(Use Spawn Probability=false) {enabled=false}
   &SpawnBurst_Instantaneous(Spawn Count=6, Spawn Group=1, Use Spawn Probability=false, UseLoopCountLimit=false)
  ]
  particleSpawnScript [
   &VC_PaintedVariantsOverride(
    Black=<rgba 0.0194 0.0194 0.02 1>
    BurntSienna=<rgba 0.14 0.0257 0.0028 1>
    Cobalt=<rgba 0.0198 0.1192 1.3 1>
    Crimson=<rgba 0.5 0.005 0.0132 1>
    DefaultColor=<rgba 1 0.2 0 1>
    ForestGreen=<rgba 0 0.4 0 1>
    Grey=<rgba 0.17 0.17 0.17 1>
    Lime=<rgba 0.2475 0.45 0 1>
    Orange=<rgba 0.7 0.0817 0 1>
    Pink=<rgba 0.643 0.1714 0.7 1>
    Purple=<rgba 0.1066 0.052 0.52 1>
    Saffron=<rgba 0.45 0.375 0 1>
    SkyBlue=<rgba 0 0.3333 1 1>
    TitaniumWhite=<rgba 0.4 0.4 0.4 1>
   )
   /SetVariables(Particles.Tag_BrokeHive:NiagaraInt32=&SpawnGroup()) {isSetParameters=true}
   &InitializeParticle(
    Color=$Particles.Color
    Lifetime=/VehicleCosmetics/FX/Niagara/Modules/MergeEmitters/VC_FloatBySpawnGroup(
     SpawnGroup or ExecIndex:NiagaraInt32=0
     CustomInt:NiagaraInt32=0
     MaximumSpawnGroups:NiagaraInt32=2
     Float SG 0:NiagaraFloat=&RandomRangeFloat(Minimum=0.3, Maximum=0.8)
     Float SG 1:NiagaraFloat=&RandomRangeFloat(Minimum=0.3365, Maximum=0.4425)
     Float SG 2:NiagaraFloat=0
     Float SG 3:NiagaraFloat=0
     Float SG 4:NiagaraFloat=0
     Float SG 5:NiagaraFloat=0
    )
    Mesh Renderer Array Visibility Mode=::NewEnumerator3
    Mesh Renderer Info=DataInterface() {propertyValues {MeshRenderer=&self:BrokeHive_0.NiagaraMeshRendererProperties_0}}
    Mesh Scale Mode=::NewEnumerator3
    Mesh Uniform Scale=&Multiply_Float(A=0.56, B=3)
    Position Offset=&AddVectorToPosition(
     Position=<-90 0 -42>
     Vector=&AddVectorToPosition(Position=$User.DTCarOffset, Vector=$User.SingleBoostPos)
    )
    Position Offset Coordinate Space=::Local
    Sprite Rotation Angle=&RandomRangeFloat(Minimum=-1, Maximum=1)
    Sprite Rotation Mode=::NewEnumerator1
    Sprite Size Mode=::NewEnumerator3
    Uniform Sprite Size=&RandomRangeFloat(Minimum=100, Maximum=200)
    UsePositionOffset=true
    Write Material Random=true
   )
   &AddVelocity(Velocity=&MaskVectorBySpawnGroup(
     Masked Vector=&RandomRangeVector(Minimum=<-70 30 100>, Maximum=<-10 -30 60>)
     Passthrough Vector=<1 1 1>
    ))
   &AddVelocity(
    Velocity=&MaskVectorBySpawnGroup(
     Masked Vector=&RandomRangeVector(Minimum=<-100 230 230>, Maximum=<-230 -230 -230>)
     Passthrough Vector=<1 1 1>
     Spawn Group Mask=1
    )
    Velocity Speed Scale=2
   ) {module=AddVelocity006}
   &ShapeLocation(MaskSpawnGroup=false, Override Seed=false, Sphere Radius=10, UseOwnerScale=true)
   &InitialMeshOrientation(Use Rotation Vector=false)
  ]
  particleUpdateScript [
   &ParticleState()
   &SolveForcesAndVelocity(Clamp Velocity=false, Limit Acceleration=false)
   &ScaleColor(
    Scale Alpha=&VC_FloatBySpawnGroup(
     SpawnGroup or ExecIndex:NiagaraInt32=0
     CustomInt:NiagaraInt32=0
     MaximumSpawnGroups:NiagaraInt32=2
     Float SG 0:NiagaraFloat=&FloatFromCurve(FloatCurve=DataInterface() {propertyValues {Curve {keys [<at 0.45 1>,<at 1 0>]}}})
     Float SG 1:NiagaraFloat=&FloatFromCurve(FloatCurve=DataInterface() {propertyValues {Curve {keys [<at 0 1.25>,<at 1 0>]}}})
     Float SG 2:NiagaraFloat=0
     Float SG 3:NiagaraFloat=0
     Float SG 4:NiagaraFloat=0
     Float SG 5:NiagaraFloat=0
    )
    Scale RGB=<2 2 2>
    ScaleA=true
    ScaleRGB=true
   )
  ]
  rendererClasses [&NiagaraSpriteRendererProperties, &NiagaraMeshRendererProperties]
  renderers [
   {rendererClass=&NiagaraSpriteRendererProperties}
   {rendererClass=&NiagaraMeshRendererProperties, rendererIndex=1}
  ]
 }
 HoneyBeesHive {
  emitterUpdateScript [
   &EmitterState(
    Inactive Response=::NewEnumerator1
    Life Cycle Mode=::NewEnumerator1
    Loop Behavior=::NewEnumerator1
    Loop Duration Mode=::NewEnumerator1
    UseLoopDelay=false
   )
   &SpawnBurst_Instantaneous(Use Spawn Probability=false, UseLoopCountLimit=false)
   &SpawnRate(SpawnGroup=1, SpawnRate=6, Use Spawn Probability=false) {enabled=false}
  ]
  particleSpawnScript [
   &VC_PaintedVariantsOverride(
    Black=<rgba 0.0194 0.0194 0.02 1>
    BurntSienna=<rgba 0.14 0.0257 0.0028 1>
    Cobalt=<rgba 0.0198 0.1192 1.3 1>
    Crimson=<rgba 0.5 0.005 0.0132 1>
    DefaultColor=<rgba 1 0.2 0 1>
    ForestGreen=<rgba 0 0.4 0 1>
    Grey=<rgba 0.17 0.17 0.17 1>
    Lime=<rgba 0.2475 0.45 0 1>
    Orange=<rgba 0.7 0.0817 0 1>
    Pink=<rgba 0.643 0.1714 0.7 1>
    Purple=<rgba 0.1066 0.052 0.52 1>
    Saffron=<rgba 0.45 0.375 0 1>
    SkyBlue=<rgba 0 0.3333 1 1>
    TitaniumWhite=<rgba 0.4 0.4 0.4 1>
   )
   &InitializeParticle(
    Color=$Particles.Color
    Lifetime=0.15
    Mesh Scale=&Multiply_VectorByFloat(Vector=<0.66 0.66 0.66>, Float=3.3673)
    Mesh Scale Mode=::NewEnumerator4
    Position Offset=&AddVectorToPosition(
     Position=<-90 0 -42>
     Vector=&AddVectorToPosition(Position=$User.DTCarOffset, Vector=$User.SingleBoostPos)
    )
    Position Offset Coordinate Space=::Local
    Sprite Rotation Mode=::NewEnumerator2
    Sprite Size Mode=::NewEnumerator3
    UsePositionOffset=true
    Write Material Random=true
   )
   &AddVelocity(Velocity=/VehicleCosmetics/FX/Niagara/Modules/MergeEmitters/VC_VectorBySpawnGroup(
     SpawnGroup or ExecIndex:NiagaraInt32=0
     CustomInt:NiagaraInt32=0
     MaximumSpawnGroups:NiagaraInt32=2
     Vector Group 0:Vector3f=<0 0 0>
     Vector Group 1:Vector3f=<10.0131 0 0>
     Vector Group 2:Vector3f=<0 0 0>
     Vector Group 3:Vector3f=<0 0 0>
     Vector Group 4:Vector3f=<0 0 0>
     Vector Group 5:Vector3f=<0 0 0>
    ))
  ]
  particleUpdateScript [
   &ParticleState()
   &ScaleMeshSize(Scale Factor=&VectorFromFloat(Value=&FloatFromCurve(FloatCurve=DataInterface() {propertyValues {Curve {keys [<at 0 1>,<at 1 1.2>]}}})))
   &SolveForcesAndVelocity(Clamp Velocity=false, Limit Acceleration=false)
   &ScaleColor(Scale Alpha=1.5, Scale RGB=<2 2 2>, ScaleA=true, ScaleRGB=true)
  ]
  rendererClasses [&NiagaraSpriteRendererProperties, &NiagaraMeshRendererProperties]
  renderers [
   {rendererClass=&NiagaraSpriteRendererProperties}
   {rendererClass=&NiagaraMeshRendererProperties, rendererIndex=1}
  ]
 }
 HoneyExplosion001 {
  emitterUpdateScript [
   &EmitterState(Life Cycle Mode=::NewEnumerator1, Loop Behavior=::NewEnumerator1, Loop Delay=0.1, UseLoopDelay=true)
   &SpawnBurst_Instantaneous(
    Spawn Count=&RandomRangeInt(Minimum=5, Maximum=10)
    Use Spawn Probability=false
    UseLoopCountLimit=false
   )
   &SpawnBurst_Instantaneous(
    Spawn Count=&RandomRangeInt(Minimum=1, Maximum=2)
    Spawn Group=1
    Use Spawn Probability=false
    UseLoopCountLimit=false
   ) {module=SpawnBurst_Instantaneous001}
   &SpawnRate(SpawnRate=50, Use Spawn Probability=false) {enabled=false}
  ]
  particleSpawnScript [
   &VC_PaintedVariantsOverride(
    Black=/VehicleCosmetics/FX/Niagara/Modules/MergeEmitters/VC_Dynamic_ColorBySpawnGroup(
     Color SG 0:LinearColor=<rgba 0.0191 0.0191 0.0191 1>
     MaximumSpawnGroups:NiagaraInt32=2
     Color SG 1:LinearColor=<rgba 0.1823 0.1823 0.2083 1>
     Color SG 2:LinearColor=<rgba 1 1 1 1>
     Color SG 3:LinearColor=<rgba 1 1 1 1>
     Color SG 4:LinearColor=<rgba 1 1 1 1>
    )
    BurntSienna=&VC_Dynamic_ColorBySpawnGroup(
     Color SG 0:LinearColor=<rgba 0.08 0.0166 0.0024 2>
     MaximumSpawnGroups:NiagaraInt32=2
     Color SG 1:LinearColor=<rgba 0.2031 0.039 0.0098 2>
     Color SG 2:LinearColor=<rgba 1 1 1 1>
     Color SG 3:LinearColor=<rgba 1 1 1 1>
     Color SG 4:LinearColor=<rgba 1 1 1 1>
    )
    Cobalt=&VC_Dynamic_ColorBySpawnGroup(
     Color SG 0:LinearColor=<rgba 0.0094 0.0478 0.47 2>
     MaximumSpawnGroups:NiagaraInt32=2
     Color SG 1:LinearColor=<rgba 0.0034 0.1657 1 2>
     Color SG 2:LinearColor=<rgba 1 1 1 1>
     Color SG 3:LinearColor=<rgba 1 1 1 1>
     Color SG 4:LinearColor=<rgba 1 1 1 1>
    )
    Crimson=&VC_Dynamic_ColorBySpawnGroup(
     Color SG 0:LinearColor=<rgba 0.28 0.0084 0.0129 1>
     MaximumSpawnGroups:NiagaraInt32=2
     Color SG 1:LinearColor=<rgba 0.5469 0 0.0148 1>
     Color SG 2:LinearColor=<rgba 1 1 1 1>
     Color SG 3:LinearColor=<rgba 1 1 1 1>
     Color SG 4:LinearColor=<rgba 1 1 1 1>
    )
    DefaultColor=&VC_Dynamic_ColorBySpawnGroup(
     Color SG 0:LinearColor=<rgba 0.5 0.1362 0.015 1>
     MaximumSpawnGroups:NiagaraInt32=2
     Color SG 1:LinearColor=<rgba 1 0.33 0 1>
     Color SG 2:LinearColor=<rgba 1 1 1 1>
     Color SG 3:LinearColor=<rgba 1 1 1 1>
     Color SG 4:LinearColor=<rgba 1 1 1 1>
    )
    ForestGreen=&VC_Dynamic_ColorBySpawnGroup(
     Color SG 0:LinearColor=<rgba 0.0021 0.105 0.0107 1>
     MaximumSpawnGroups:NiagaraInt32=2
     Color SG 1:LinearColor=<rgba 0 0.3177 0.0161 1>
     Color SG 2:LinearColor=<rgba 1 1 1 1>
     Color SG 3:LinearColor=<rgba 1 1 1 1>
     Color SG 4:LinearColor=<rgba 1 1 1 1>
    )
    Gold:LinearColor=<rgba 0.6094 0.1987 0 1>
    Grey=&VC_Dynamic_ColorBySpawnGroup(
     Color SG 0:LinearColor=&RandomRangeLinearColor(Minimum=<rgba 0.0788 0.0788 0.09 1>, Maximum=<rgba 0.0525 0.0525 0.06 1>)
     MaximumSpawnGroups:NiagaraInt32=2
     Color SG 1:LinearColor=<rgba 0.1823 0.1823 0.2083 1>
     Color SG 2:LinearColor=<rgba 1 1 1 1>
     Color SG 3:LinearColor=<rgba 1 1 1 1>
     Color SG 4:LinearColor=<rgba 1 1 1 1>
    )
    Lime=&VC_Dynamic_ColorBySpawnGroup(
     Color SG 0:LinearColor=<rgba 0.0811 0.17 0.0034 1>
     MaximumSpawnGroups:NiagaraInt32=2
     Color SG 1:LinearColor=<rgba 0.3486 0.6042 0 1>
     Color SG 2:LinearColor=<rgba 1 1 1 1>
     Color SG 3:LinearColor=<rgba 1 1 1 1>
     Color SG 4:LinearColor=<rgba 1 1 1 1>
    )
    Orange=&VC_Dynamic_ColorBySpawnGroup(
     Color SG 0:LinearColor=<rgba 0.29 0.0439 0.0087 1>
     MaximumSpawnGroups:NiagaraInt32=2
     Color SG 1:LinearColor=<rgba 0.6724 0.1374 0 1>
     Color SG 2:LinearColor=<rgba 1 1 1 1>
     Color SG 3:LinearColor=<rgba 1 1 1 1>
     Color SG 4:LinearColor=<rgba 1 1 1 1>
    )
    Pink=&VC_Dynamic_ColorBySpawnGroup(
     Color SG 0:LinearColor=<rgba 0.3092 0.105 0.35 1>
     MaximumSpawnGroups:NiagaraInt32=2
     Color SG 1:LinearColor=<rgba 0.8646 0.3095 0.7961 1>
     Color SG 2:LinearColor=<rgba 1 1 1 1>
     Color SG 3:LinearColor=<rgba 1 1 1 1>
     Color SG 4:LinearColor=<rgba 1 1 1 1>
    )
    Purple=&VC_Dynamic_ColorBySpawnGroup(
     Color SG 0:LinearColor=<rgba 0.0685 0.0265 0.53 1>
     MaximumSpawnGroups:NiagaraInt32=2
     Color SG 1:LinearColor=<rgba 0.1717 0.0918 0.7552 1>
     Color SG 2:LinearColor=<rgba 1 1 1 1>
     Color SG 3:LinearColor=<rgba 1 1 1 1>
     Color SG 4:LinearColor=<rgba 1 1 1 1>
    )
    Saffron=<rgba 0.3 0.2515 0.009 1>
    SkyBlue=&VC_Dynamic_ColorBySpawnGroup(
     Color SG 0:LinearColor=<rgba 0 0.1425 0.45 1>
     MaximumSpawnGroups:NiagaraInt32=2
     Color SG 1:LinearColor=<rgba 0.0787 0.5836 1 1>
     Color SG 2:LinearColor=<rgba 1 1 1 1>
     Color SG 3:LinearColor=<rgba 1 1 1 1>
     Color SG 4:LinearColor=<rgba 1 1 1 1>
    )
    TitaniumWhite=&VC_Dynamic_ColorBySpawnGroup(
     Color SG 0:LinearColor=<rgba 0.2072 0.2074 0.22 1>
     MaximumSpawnGroups:NiagaraInt32=2
     Color SG 1:LinearColor=<rgba 0.6327 0.6333 0.6719 1>
     Color SG 2:LinearColor=<rgba 1 1 1 1>
     Color SG 3:LinearColor=<rgba 1 1 1 1>
     Color SG 4:LinearColor=<rgba 1 1 1 1>
    )
   )
   /SetVariables(Particles.Tag_HoneyExp:NiagaraInt32=&SpawnGroup()) {isSetParameters=true}
   &InitializeParticle(
    Color=$Particles.Color
    Lifetime=&VC_FloatBySpawnGroup(
     SpawnGroup or ExecIndex:NiagaraInt32=0
     CustomInt:NiagaraInt32=0
     MaximumSpawnGroups:NiagaraInt32=2
     Float SG 0:NiagaraFloat=&RandomRangeFloat(Minimum=0.25, Maximum=0.6)
     Float SG 1:NiagaraFloat=&RandomRangeFloat(Minimum=0.3, Maximum=0.5)
     Float SG 2:NiagaraFloat=0
     Float SG 3:NiagaraFloat=0
     Float SG 4:NiagaraFloat=0
     Float SG 5:NiagaraFloat=0
    )
    Position Offset=&AddVectorToPosition(
     Position=<-90 0 -42>
     Vector=&AddVectorToPosition(Position=$User.DTCarOffset, Vector=$User.SingleBoostPos)
    )
    Position Offset Coordinate Space=::Local
    Sprite Rotation Mode=::NewEnumerator2
    Sprite Size Mode=::NewEnumerator1
    Uniform Sprite Size Max=&VC_FloatBySpawnGroup(
     SpawnGroup or ExecIndex:NiagaraInt32=0
     CustomInt:NiagaraInt32=0
     MaximumSpawnGroups:NiagaraInt32=2
     Float SG 0:NiagaraFloat=14
     Float SG 1:NiagaraFloat=&Multiply_Float(A=75, B=3)
     Float SG 2:NiagaraFloat=0
     Float SG 3:NiagaraFloat=0
     Float SG 4:NiagaraFloat=0
     Float SG 5:NiagaraFloat=0
    )
    Uniform Sprite Size Min=&VC_FloatBySpawnGroup(
     SpawnGroup or ExecIndex:NiagaraInt32=0
     CustomInt:NiagaraInt32=0
     MaximumSpawnGroups:NiagaraInt32=2
     Float SG 0:NiagaraFloat=6
     Float SG 1:NiagaraFloat=&Multiply_Float(A=40, B=3)
     Float SG 2:NiagaraFloat=0
     Float SG 3:NiagaraFloat=0
     Float SG 4:NiagaraFloat=0
     Float SG 5:NiagaraFloat=0
    )
    UsePositionOffset=true
    Write Material Random=true
   )
   &AddVelocity(Velocity=&VC_VectorBySpawnGroup(
     SpawnGroup or ExecIndex:NiagaraInt32=0
     CustomInt:NiagaraInt32=0
     MaximumSpawnGroups:NiagaraInt32=2
     Vector Group 0:Vector3f=&RandomRangeVector(Minimum=<-190 -230 200>, Maximum=<-230 230 150>)
     Vector Group 1:Vector3f=&RandomRangeVector(Minimum=<-20 30 80>, Maximum=<-10 -30 60>)
     Vector Group 2:Vector3f=<0 0 0>
     Vector Group 3:Vector3f=<0 0 0>
     Vector Group 4:Vector3f=<0 0 0>
     Vector Group 5:Vector3f=<0 0 0>
    ))
   &ShapeLocation(
    MaskSpawnGroup=false
    Override Seed=false
    Sphere Radius=&VC_FloatBySpawnGroup(
     SpawnGroup or ExecIndex:NiagaraInt32=0
     CustomInt:NiagaraInt32=0
     MaximumSpawnGroups:NiagaraInt32=2
     Float SG 0:NiagaraFloat=&RandomRangeFloat(Minimum=0, Maximum=10)
     Float SG 1:NiagaraFloat=0
     Float SG 2:NiagaraFloat=0
     Float SG 3:NiagaraFloat=0
     Float SG 4:NiagaraFloat=0
     Float SG 5:NiagaraFloat=0
    )
    UseOwnerScale=true
   )
  ]
  particleUpdateScript [
   &ParticleState()
   &ScaleSpriteSize(
    Scale Sprite Size Mode=::NewEnumerator0
    Uniform Scale Factor=&VC_FloatBySpawnGroup(
     SpawnGroup or ExecIndex:NiagaraInt32=0
     CustomInt:NiagaraInt32=0
     MaximumSpawnGroups:NiagaraInt32=2
     Float SG 0:NiagaraFloat=&FloatFromCurve(FloatCurve=DataInterface() {propertyValues {Curve {keys [{tangentMode=RCTM_User, time=0.33, value=1}, {tangentMode=RCTM_User, time=1, value=0.25}]}}})
     Float SG 1:NiagaraFloat=&FloatFromCurve(FloatCurve=DataInterface() {propertyValues {Curve {keys [<at 0 1>,<at 1 1.2>]}}})
     Float SG 2:NiagaraFloat=0
     Float SG 3:NiagaraFloat=0
     Float SG 4:NiagaraFloat=0
     Float SG 5:NiagaraFloat=0
    )
   )
   &SolveForcesAndVelocity(Clamp Velocity=false, Limit Acceleration=false)
   &DynamicMaterialParameters(
    Index 0 Param 1=1
    Index 0 Param 3=1
    Index 0 Param 4=2
    Module.Index0Float0Write=true
    Module.Index0Float1Write=true
    Module.Index0Float2Write=true
    Module.Index0Float3Write=true
   )
   &ScaleColor(Scale RGB=<1.7 1.7 1.7>, ScaleA=true, ScaleRGB=true)
   &SubUVAnimation(
    "Sprite Renderer "=DataInterface() {propertyValues {SpriteRenderer=&self:HoneyExplosion001_0.NiagaraSpriteRendererProperties_1}}
    UseEndFrame=false
    UseStartFrame=false
   )
  ]
  rendererClasses [&NiagaraSpriteRendererProperties]
  renderers [
   {rendererClass=&NiagaraSpriteRendererProperties}
   {rendererClass=&NiagaraSpriteRendererProperties, rendererIndex=1}
  ]
 }
}
summary {
 systemName=NS_Boost_Honeybees_Explosion
 emitters [
  {emitterName=BrokeHive, rendererClasses [&NiagaraSpriteRendererProperties, &NiagaraMeshRendererProperties]}
  {emitterName=HoneyBeesHive, rendererClasses [&NiagaraSpriteRendererProperties, &NiagaraMeshRendererProperties]}
  {emitterName=HoneyExplosion001, rendererClasses [&NiagaraSpriteRendererProperties]}
 ]
 userVariables [
  User.bIsFrontendPreview {type=NiagaraBool}
  User.bIsPainted {type=NiagaraBool}
  User.ColorVariantID {type=NiagaraInt32}
  User.DTCarOffset {type=NiagaraPosition, description="", defaultValue {}}
  User.PaintedColor {type=LinearColor, description="", defaultValue {struct=LinearColor, value=<rgba 0.8073 0.2503 0 1>}}
  User.PaintedScalar {type=NiagaraFloat, defaultValue=1, description=""}
  User.PortCount {type=NiagaraInt32, defaultValue=1}
  User.Shine {type=NiagaraFloat, defaultValue=5, description=""}
  User.SingleBoostPos {type=Vector3f, description="", defaultValue {struct=Vector3f, value=<0 0 0>}}
 ]
}
