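# Niagara system for the Honeybees vehicle boost cosmetic, drawing a beehive that spills honey behind the car.
# A jittering hive, falling SubUV droplets, stretching mesh swirls and a source-following ribbon trail are offset from the car and boost-position user parameters.
# Every emitter takes its color from painted-variant overrides, with the swirl adding two extra color sets scaled over life.
# Curves drive size, velocity and material parameters, with user parameters feeding specular and shine.
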NiagaraSystem /VehicleCosmetics/Boosts/Boost_Honeybees/FX/NS_Boost_Honeybees_Revision
dependencies {
 usedDataInterfaces [
  &NiagaraDataInterfaceArrayFloat
  &NiagaraDataInterfaceCurve
  &NiagaraDataInterfaceVectorCurve
  &NiagaraDataInterfaceSpriteRendererInfo
  &NiagaraDataInterfaceSkeletalMesh
 ]
 usedDynamicInputs []
 usedModules [
  &self:SystemSpawnScript.SystemScriptSource.SystemScriptGraph.NiagaraNodeAssignment_0.SetVariables_#1
  &SystemState
 ]
}
emitters {
 NE_BeeHive {
  emitterUpdateScript [
   &EmitterState(
    Inactive Response=::NewEnumerator1
    Life Cycle Mode=::NewEnumerator1
    Loop Behavior=::NewEnumerator1
    Loop Duration Mode=::NewEnumerator1
    UseLoopDelay=false
   )
   &SpawnBurst_Instantaneous(Use Spawn Probability=false, UseLoopCountLimit=false)
   &SpawnRate(SpawnGroup=1, SpawnRate=6, Use Spawn Probability=false) {enabled=false}
  ]
  particleSpawnScript [
   &VC_PaintedVariantsOverride(
    Black=<rgba 0.0194 0.0194 0.02 1>
    BurntSienna=<rgba 0.14 0.0257 0.0028 1>
    Cobalt=<rgba 0.0198 0.1192 1.3 1>
    Crimson=<rgba 0.5 0.005 0.0132 1>
    DefaultColor=<rgba 1 0.2 0 1>
    ForestGreen=<rgba 0 0.4 0 1>
    Grey=<rgba 0.17 0.17 0.17 1>
    Lime=<rgba 0.2475 0.45 0 1>
    Orange=<rgba 0.7 0.0817 0 1>
    Pink=<rgba 0.643 0.1714 0.7 1>
    Purple=<rgba 0.1066 0.052 0.52 1>
    Saffron=<rgba 0.45 0.375 0 1>
    SkyBlue=<rgba 0 0.3333 1 1>
    TitaniumWhite=<rgba 0.4 0.4 0.4 1>
   )
   &InitializeParticle(
    Color=$Particles.Color
    Lifetime=/VehicleCosmetics/FX/Niagara/Modules/MergeEmitters/VC_FloatBySpawnGroup(
     SpawnGroup or ExecIndex:NiagaraInt32=0
     CustomInt:NiagaraInt32=0
     MaximumSpawnGroups:NiagaraInt32=2
     Float SG 0:NiagaraFloat=1
     Float SG 1:NiagaraFloat=0.4
     Float SG 2:NiagaraFloat=0
     Float SG 3:NiagaraFloat=0
     Float SG 4:NiagaraFloat=0
     Float SG 5:NiagaraFloat=0
    )
    Mesh Scale=&Multiply_VectorByFloat(Vector=<0.66 0.66 0.728>, Float=2.85)
    Mesh Scale Mode=::NewEnumerator4
    Position Offset=&AddVectorToPosition(
     Position=<-90 0 -35>
     Vector=&AddVectorToPosition(Position=$User.DTCarOffset, Vector=$User.SingleBoostPos)
    )
    Position Offset Coordinate Space=::Local
    Sprite Rotation Mode=::NewEnumerator2
    Sprite Size Mode=::NewEnumerator3
    UsePositionOffset=true
    Write Material Random=true
   )
   &AddVelocity(Velocity=<0 0 0>)
  ]
  particleUpdateScript [
   &ParticleState(Kill Particles When Lifetime Has Elapsed=false, Let Infinitely Lived Particles Die When Emitter Deactivates=true)
   &SolveForcesAndVelocity(Clamp Velocity=false, Limit Acceleration=false)
   &ScaleColor(Scale RGB=<2 2 2>, ScaleA=true, ScaleRGB=true)
   &ShapeLocation(
    MaskSpawnGroup=false
    Offset=&MakeVector(
     X=&Sine(Normalized Angle=$Emitter.Age, Period=0.6, Scale=1, Bias=0)
     Y=&Cosine(Normalized Angle=$Emitter.Age, Period=0.6, Scale=1, Bias=0)
     Z=&Sine(Normalized Angle=$Emitter.Age, Period=0.25, Scale=1, Bias=0)
    )
    Offset Mode=::NewEnumerator0
    Override Seed=false
    Shape Origin=$Particles.Initial.Position
    Sphere Radius=0
    UseOwnerScale=true
   )
  ]
  rendererClasses [&NiagaraSpriteRendererProperties, &NiagaraMeshRendererProperties]
  renderers [
   {rendererClass=&NiagaraSpriteRendererProperties}
   {rendererClass=&NiagaraMeshRendererProperties, rendererIndex=1}
  ]
 }
 NE_HoneyDrops {
  emitterSpawnScript [/SetVariables(Emitter.SPEC_POWER:NiagaraFloat=&SelectFloatFromArray(
     Direct Array Index=$User.ColorVariantID
     Float Selection Array=DataInterface() {propertyValues {FloatData [25,1,50,40,40,45,45,30,35,30,30,20,50,25,75,0]}}
    )) {isSetParameters=true}]
  emitterUpdateScript [
   &EmitterState()
   &SpawnRate(SpawnRate=35, Use Spawn Probability=false)
  ]
  particleSpawnScript [
   &VC_PaintedVariantsOverride(
    Black=<rgba 0.008 0.008 0.008 1>
    BurntSienna=<rgba 0.08 0.0166 0.0024 2>
    Cobalt=<rgba 0.0094 0.0478 0.47 2>
    Crimson=<rgba 0.28 0.0084 0.0129 1>
    DefaultColor=<rgba 0.5 0.1362 0.015 1>
    ForestGreen=<rgba 0.0021 0.105 0.0107 1>
    Gold:LinearColor=<rgba 0.6094 0.1987 0 1>
    Grey=&RandomRangeLinearColor(Minimum=<rgba 0.0788 0.0788 0.09 1>, Maximum=<rgba 0.0525 0.0525 0.06 1>)
    Lime=<rgba 0.0811 0.17 0.0034 1>
    Orange=<rgba 0.29 0.0439 0.0087 1>
    Pink=<rgba 0.3092 0.105 0.35 1>
    Purple=<rgba 0.0685 0.0265 0.53 1>
    Saffron=<rgba 0.3 0.2515 0.009 1>
    SkyBlue=<rgba 0 0.1425 0.45 1>
    TitaniumWhite=<rgba 0.2072 0.2074 0.22 1>
   )
   &InitializeParticle(
    Color=$Particles.Color
    Lifetime Max=0.8
    Lifetime Min=0.65
    Lifetime Mode=::NewEnumerator1
    Mesh Scale=&Multiply_VectorByFloat(Vector=<0.85 0.85 0.6>, Float=3.4)
    Mesh Scale Mode=::NewEnumerator4
    Position Offset=&AddVectorToPosition(
     Position=<-150 0 -15>
     Vector=&AddVectorToPosition(Position=$User.DTCarOffset, Vector=$User.SingleBoostPos)
    )
    Position Offset Coordinate Space=::Local
    Sprite Rotation Angle=&RandomRangeFloat(Minimum=-1, Maximum=1)
    Sprite Rotation Mode=::NewEnumerator1
    Sprite Size=&RandomRangeVector2D(
     Minimum=&Multiply_Vector2D_ByFloat(Float=1.5, Vector2D=<6 6>)
     Maximum=&Multiply_Vector2D_ByFloat(Float=1.5, Vector2D=<9 9>)
    )
    Sprite Size Mode=::NewEnumerator4
    UsePositionOffset=true
    Write Material Random=true
   )
   &AddVelocity(Velocity=&RandomRangeVector(Minimum=<-55 -10 -15>, Maximum=<-30 10 -25>))
   &InitialMeshOrientation(
    Mesh Axis To Orient=<0 0 0>
    Mesh Orientation Mode=::NewEnumerator2
    Orientation Vector=<0 0 1>
    Rotation=&RandomRangeVector(Minimum=<0 0 -1>, Maximum=<0 0 1>)
    Use Rotation Vector=true
   ) {enabled=false}
   &Color(Color=/VehicleCosmetics/FX/Niagara/Modules/MergeEmitters/VC_Dynamic_ColorBySpawnGroup(
     Color SG 0:LinearColor=<rgba 1 0.3096 0 1>
     MaximumSpawnGroups:NiagaraInt32=2
     Color SG 1:LinearColor=<rgba 0.5729 0.2431 0.0001 1>
     Color SG 2:LinearColor=<rgba 1 1 1 1>
     Color SG 3:LinearColor=<rgba 1 1 1 1>
     Color SG 4:LinearColor=<rgba 1 1 1 1>
    )) {enabled=false}
   &ShapeLocation(MaskSpawnGroup=false, Override Seed=false, Sphere Radius=20, UseOwnerScale=true)
  ]
  particleUpdateScript [
   &ParticleState()
   &ScaleSpriteSize(
    Uniform Curve Scale=2.3
    Uniform Curve Sprite Scale=DataInterface() {propertyValues {Curve {keys [
        {arriveTangent=-1.1727, interpMode=RCIM_Cubic, leaveTangent=-1.3465, tangentMode=RCTM_Break, time=0, value=0.5}
        {arriveTangent=-0.3, interpMode=RCIM_Cubic, leaveTangent=-0.3, time=0.1702, value=0.7984}
        {arriveTangent=-0.7427, interpMode=RCIM_Cubic, tangentMode=RCTM_Break, time=1, value=0.2}
       ]}}}
   )
   &ScaleVelocity(
    Coordinate Space:ENiagaraCoordinateSpace=::Simulation
    Velocity Scale=&VectorFromCurve(
     VectorCurve=DataInterface() {propertyValues {
       XCurve {
        defaultValue=1
        keys [
         {interpMode=RCIM_Cubic, tangentMode=RCTM_Break, time=0, value=1}
         {arriveTangentWeight=1.02, interpMode=RCIM_Cubic, tangentMode=RCTM_Break, time=1, value=1}
        ]
       }
       YCurve {
        defaultValue=1
        keys [
         {interpMode=RCIM_Cubic, tangentMode=RCTM_Break, time=0, value=1}
         {arriveTangentWeight=1.02, interpMode=RCIM_Cubic, tangentMode=RCTM_Break, time=1, value=1}
        ]
       }
       ZCurve {keys [
         {interpMode=RCIM_Cubic, leaveTangent=-1.6494, tangentMode=RCTM_Break, time=0, value=0}
         {arriveTangent=0.0661, arriveTangentWeight=1.02, interpMode=RCIM_Cubic, tangentMode=RCTM_Break, time=1, value=1}
        ]}
      }}
     Scale Curve=&RandomRangeVector(Minimum=<2 2 7>, Maximum=<4 3 4>)
    )
   )
   &SolveForcesAndVelocity(Clamp Velocity=false, Limit Acceleration=false)
   &SubUVAnimation(
    "Sprite Renderer "=DataInterface() {propertyValues {SpriteRenderer=&self:NE_HoneyDrops_0.NiagaraSpriteRendererProperties_0}}
    UseEndFrame=false
    UseStartFrame=false
   )
   &DynamicMaterialParameters(
    Index 0 Param 1=$User.SPECULAR_FORCE
    Index 0 Param 3=1
    Index 0 Param 4=1.5
    Module.Index0Float0Write=true
    Module.Index0Float1Write=true
    Module.Index0Float2Write=true
    Module.Index0Float3Write=true
   )
   &ScaleColor(ScaleA=true, ScaleRGB=true)
  ]
  rendererClasses [&NiagaraSpriteRendererProperties]
  renderers [{rendererClass=&NiagaraSpriteRendererProperties}]
 }
 NE_HoneySwirl {
  emitterUpdateScript [
   &EmitterState()
   &SpawnRate(SpawnRate=12, Use Spawn Probability=false)
  ]
  particleSpawnScript [
   &VC_PaintedVariantsOverride(
    Black=<rgba 0.035 0.035 0.035 1>
    BurntSienna=<rgba 0.17 0.034 0 2>
    Cobalt=<rgba 0.02 0.0982 0.5 2>
    Crimson=<rgba 1 0.02 0.02 1>
    DefaultColor=<rgba 1 0.28 0.1 1>
    ForestGreen=<rgba 0 0.6 0 1>
    Grey=<rgba 0.175 0.175 0.2 0.5>
    Lime=<rgba 0.35 0.7 0 1>
    Orange=<rgba 0.55 0.1152 0.0385 1>
    Pink=<rgba 0.7 0.21 0.6395 1>
    Purple=<rgba 0.1528 0.0817 0.6719 1>
    R1 Black:LinearColor=<rgba 0.007 0.007 0.007 5>
    R1 BurntSienna:LinearColor=<rgba 0.06 0.0134 0.0018 1>
    R1 Cobalt:LinearColor=<rgba 0 0.0373 0.32 1>
    R1 Crimson:LinearColor=<rgba 0.2 0.01 0.0132 1>
    R1 DefaultColor=<rgba 0.33 0.0844 0.0165 1>
    R1 ForestGreen:LinearColor=<rgba 0 0.08 0.004 1>
    R1 Grey:LinearColor=<rgba 0.05 0.05 0.05 2>
    R1 Lime:LinearColor=<rgba 0.0507 0.12 0.0012 2>
    R1 Orange:LinearColor=<rgba 0.28 0.0398 0.0028 1>
    R1 Pink:LinearColor=<rgba 0.1527 0.058 0.2 1.5>
    R1 Purple:LinearColor=<rgba 0.0464 0.016 0.32 1.5>
    R1 Saffron:LinearColor=<rgba 0.2 0.1667 0 1>
    R1 SkyBlue:LinearColor=<rgba 0 0.1 0.3 1>
    R1 TitaniumWhite:LinearColor=<rgba 0.12 0.12 0.12 2>
    R2 Black:LinearColor=<rgba 0.02 0.02 0.02 1>
    R2 BurntSienna:LinearColor=<rgba 0.2 0.0383 0.002 1>
    R2 Cobalt:LinearColor=<rgba 0.085 0.2762 0.85 0.65>
    R2 Crimson:LinearColor=<rgba 0.8 0.024 0.0369 1>
    R2 DefaultColor:LinearColor=<rgba 1 0.25 0 1>
    R2 ForestGreen:LinearColor=<rgba 0.006 0.3 0.0305 0.5>
    R2 Grey:LinearColor=<rgba 0.225 0.225 0.225 0.5>
    R2 Lime:LinearColor=<rgba 0.202 0.4 0.004 1>
    R2 Orange:LinearColor=<rgba 0.9 0.1512 0.036 1>
    R2 Pink:LinearColor=<rgba 0.8 0.24 0.5964 0.9>
    R2 Purple:LinearColor=<rgba 0.1871 0.122 0.61 1>
    R2 Saffron:LinearColor=<rgba 0.6 0.51 0.06 1>
    R2 SkyBlue:LinearColor=<rgba 0.05 0.3494 0.5 1>
    R2 TitaniumWhite:LinearColor=<rgba 0.5 0.5 0.5 1>
    Saffron=<rgba 0.9 0.7005 0.045 1>
    SkyBlue=<rgba 0.0325 0.3412 0.65 1>
    TitaniumWhite=<rgba 0.4905 0.491 0.5208 1>
    Use R1 Painted Overrides=true
    Use R2 Painted Overrides=true
   )
   &InitializeParticle(
    Color=$Particles.Color
    Lifetime=0.45
    Mesh Scale=&Multiply_VectorByFloat(Vector=<0.85 0.85 0.6>, Float=3.4)
    Mesh Scale Mode=::NewEnumerator4
    Position Offset=&AddVectorToPosition(
     Position=<-125 0 -16>
     Vector=&AddVectorToPosition(Position=$User.DTCarOffset, Vector=$User.SingleBoostPos)
    )
    Position Offset Coordinate Space=::Local
    Sprite Size Mode=::NewEnumerator3
    UsePositionOffset=true
    Write Material Random=true
   )
   &AddVelocity(Velocity=<-5 0 0>)
   &InitialMeshOrientation(
    Mesh Axis To Orient=<0 0 0>
    Mesh Orientation Mode=::NewEnumerator2
    Orientation Vector=<0 0 1>
    Rotation=&RandomRangeVector(Minimum=<0 0 -1>, Maximum=<0 0 1>)
    Use Rotation Vector=true
   )
  ]
  particleUpdateScript [
   &ParticleState()
   &CameraOffset(Camera Offset Amount=&VC_FloatBySpawnGroup(
     SpawnGroup or ExecIndex:NiagaraInt32=0
     CustomInt:NiagaraInt32=0
     MaximumSpawnGroups:NiagaraInt32=3
     Float SG 0:NiagaraFloat=0
     Float SG 1:NiagaraFloat=0
     Float SG 2:NiagaraFloat=0
     Float SG 3:NiagaraFloat=0
     Float SG 4:NiagaraFloat=0
     Float SG 5:NiagaraFloat=0
    )) {enabled=false}
   &UpdateMeshOrientation(Rotation Vector=&RandomRangeVector(Minimum=<0 0 -0.5>, Maximum=<0 0 0.5>))
   &ScaleMeshSize(Scale Factor=&VectorFromCurve(
     VectorCurve=DataInterface() {propertyValues {
       XCurve {defaultValue=0.1, keys [<at 0 0.5>,<at 0.7 0.8>,<at 1 0.8>]}
       YCurve {keys [<at 0 0.35>, {arriveTangent=0.65, leaveTangent=0.65, time=0.45, value=0.8}, <at 1 1>]}
       ZCurve {defaultValue=0.35, keys [<at 0 0.5>, <at 0.25 1>, {arriveTangent=-0.4627, interpMode=RCIM_Cubic, tangentMode=RCTM_Break, time=1, value=4}]}
      }}
     Scale Curve=&RandomRangeVector(Minimum=<1 0.8 0.85>, Maximum=<0.85 1.3 1.25>)
    )) {module=ScaleMeshSize001}
   &SolveForcesAndVelocity(Clamp Velocity=false, Limit Acceleration=false)
   &DynamicMaterialParameters(
    Index 1 Param 1=&FloatFromCurve(
     FloatCurve=DataInterface() {propertyValues {Curve {
        defaultValue=1
        keys [<at 0.4 0 RCIM_Cubic>, {arriveTangent=0.5134, interpMode=RCIM_Cubic, leaveTangent=0.5134, tangentMode=RCTM_User, time=1, value=0.9}]
       }}}
     Scale Curve=0.9
    )
    Module.Index0Float0Write=true
    Module.Index0Float1Write=true
    Module.Index0Float2Write=true
    Module.Index0Float3Write=true
    Module.Index1Float0Write=true
    Module.Index1Float1Write=true
    Module.Index1Float2Write=true
    Module.Index1Float3Write=true
    Write Parameter Index 1=::NewEnumerator1
   )
   &ScaleColor(Scale Alpha=4, Scale RGB=&VectorFromFloat(Value=2.3), ScaleA=true, ScaleRGB=true)
   /SetVariables(
    Particles.Scale_R1_Color:LinearColor=&VC_DynamicScaleColor(
     Color to Scale:LinearColor=$Particles.Initial.R1 Color
     RGB Scale:Vector3f=&VectorFromFloat(Value=&FloatFromCurve(FloatCurve=DataInterface() {propertyValues {Curve {
          defaultValue=1
          keys [<at 0.3 0.85 RCIM_Cubic>, {arriveTangent=5.9286, interpMode=RCIM_Cubic, leaveTangent=5.9286, time=0.65, value=2}, <at 1 5 RCIM_Cubic>]
         }}}))
     Alpha Scale:NiagaraFloat=&FloatFromCurve(FloatCurve=DataInterface() {propertyValues {Curve {defaultValue=1, keys [<at 0 1 RCIM_Cubic>,<at 1 2.5 RCIM_Cubic>]}}}, Scale Curve=7)
    )
    Particles.Scale_R2_Color:LinearColor=&VC_DynamicScaleColor(
     Color to Scale:LinearColor=$Particles.Initial.R2 Color
     RGB Scale:Vector3f=&VectorFromFloat(Value=&FloatFromCurve(
       FloatCurve=DataInterface() {propertyValues {Curve {defaultValue=1, keys [{tangentMode=RCTM_User, time=0, value=1}, {tangentMode=RCTM_User, time=1, value=2}]}}}
       Scale Curve=0.8
      ))
     Alpha Scale:NiagaraFloat=&FloatFromCurve(
      FloatCurve=DataInterface() {propertyValues {Curve {defaultValue=1, keys [{tangentMode=RCTM_User, time=0, value=1}, {tangentMode=RCTM_User, time=1, value=1}]}}}
      Scale Curve=50
     )
    )
   ) {isSetParameters=true}
  ]
  rendererClasses [&NiagaraSpriteRendererProperties, &NiagaraMeshRendererProperties]
  renderers [
   {rendererClass=&NiagaraSpriteRendererProperties}
   {rendererClass=&NiagaraMeshRendererProperties, rendererIndex=1}
   {rendererClass=&NiagaraMeshRendererProperties, rendererIndex=2}
   {rendererClass=&NiagaraMeshRendererProperties, rendererIndex=3}
  ]
 }
 NE_HoneyTrail001 {
  emitterUpdateScript [
   &EmitterState()
   &VC_FilteredSocketCount()
   &VC_RateQualityHandler(Low Spawn Rate=45, Medium Spawn Rate=45, Use Custom Rate?=true)
   &VC_SpawnRibbon_v1(Spawn Time=-1, SpawnRate=$Emitter.Spawn Rate)
  ]
  particleSpawnScript [
   &VC_PaintedVariantsOverride(
    Black=<rgba 0.02 0.016 0.016 20>
    BurntSienna=<rgba 0.15 0.0238 0.0015 10>
    Cobalt=<rgba 0.035 0.0793 0.7 10>
    Crimson=<rgba 0.55 0.011 0.0155 10>
    DefaultColor=<rgba 0.65 0.1972 0.0325 5>
    ForestGreen=<rgba 0.0225 0.225 0.0225 10>
    Grey=<rgba 0.125 0.1063 0.1158 10>
    Lime=<rgba 0.2123 0.3 0.0075 10>
    Orange=<rgba 0.52 0.0737 0 10>
    Pink=<rgba 0.45 0.135 0.4237 10>
    Purple=<rgba 0.1274 0.0267 0.89 10>
    Saffron=<rgba 0.4 0.2733 0.02 7>
    SkyBlue=<rgba 0.055 0.1787 0.55 10>
    TitaniumWhite=<rgba 0.45 0.3825 0.3825 10>
   )
   &VC_InitializeRibbon_v1(
    Color=$Particles.Color
    Custom Offset=true
    Lifetime=0.7
    Local Vector Offset=&AddVectorToPosition(
     Position=<-90 0 -18>
     Vector=&AddVectorToPosition(Position=$User.DTCarOffset, Vector=$User.SingleBoostPos)
    )
    Offset Amount=1
    Ribbon Width=70
   )
   &AddVelocity(Velocity=<-2000 0 -800>) {module=AddVelocity001}
  ]
  particleUpdateScript [
   &ParticleState()
   &ScaleVelocity(
    Coordinate Space:ENiagaraCoordinateSpace=::Simulation
    Velocity Scale=&VectorFromCurve(VectorCurve=DataInterface() {propertyValues {
       XCurve {keys [<at 0 1>,<at 1 1>]}
       YCurve {keys [<at 0 1>,<at 1 1>]}
       ZCurve {keys [
         {arriveTangent=-1.0508, interpMode=RCIM_Cubic, leaveTangent=-1.0508, tangentMode=RCTM_User, time=0, value=0}
         {arriveTangent=0.3335, interpMode=RCIM_Cubic, leaveTangent=0.3335, tangentMode=RCTM_User, time=1, value=1}
        ]}
      }})
   )
   &InheritSourceMovement(
    Applied Position Scale=&Multiply_VectorByFloat(
     Vector=<1 1 1>
     Float=&FloatFromCurve(FloatCurve=DataInterface() {propertyValues {Curve {keys [
          {interpMode=RCIM_Cubic, leaveTangent=-0.0051, tangentMode=RCTM_Break, time=0, value=1}
          {arriveTangent=-0.247, interpMode=RCIM_Cubic, tangentMode=RCTM_Break, time=1, value=0.9}
         ]}}})
    )
    ApplyForce=false
    ApplyPosition=true
    ApplyPositionFalloff=false
   )
   &SolveForcesAndVelocity(Clamp Velocity=false, Limit Acceleration=false)
   &VC_UpdateRibbon_v1()
   &ScaleColor(
    Scale Alpha=&FloatFromCurve(FloatCurve=DataInterface() {propertyValues {Curve {keys [
         {interpMode=RCIM_Cubic, leaveTangent=16.6325, tangentMode=RCTM_Break, time=0.1, value=0}
         {interpMode=RCIM_Cubic, tangentMode=RCTM_Break, time=0.7, value=4}
        ]}}})
    Scale RGB=<1.2 1.5 1.5>
    ScaleA=true
    ScaleRGB=true
   )
   &ScaleRibbonWidth(Ribbon Width Scale=&FloatFromCurve(FloatCurve=DataInterface() {propertyValues {Curve {keys [
         <at 0 0.3>
         {arriveTangent=2.0933, interpMode=RCIM_Cubic, leaveTangent=2.0933, time=0.25, value=1.15}
         {arriveTangent=-0.595, interpMode=RCIM_Cubic, leaveTangent=-0.595, time=0.4299, value=1.2}
         {arriveTangent=-3.9496, interpMode=RCIM_Cubic, leaveTangent=-3.9496, tangentMode=RCTM_User, time=1.0063, value=0.7}
        ]}}}))
   /SetVariables("Particles.DMP_HoneyTrail;":Vector4f=&MakeVector4(
     X=&FloatFromCurve(FloatCurve=DataInterface() {propertyValues {Curve {keys [
          {interpMode=RCIM_Cubic, tangentMode=RCTM_Break, time=0, value=0.25}
          {interpMode=RCIM_Cubic, tangentMode=RCTM_Break, time=0.5, value=0.4}
          {interpMode=RCIM_Cubic, tangentMode=RCTM_Break, time=1, value=0.5}
         ]}}})
     Y=$User.Shine
     Z=1.4
     W=1.8
    )) {isSetParameters=true}
   &ScaleRibbonWidth(Ribbon Width Scale=&FloatFromCurve(FloatCurve=DataInterface() {propertyValues {Curve {keys [<at 0 0.25>,<at 0.1 0.75>,<at 1 0.2>]}}})) {enabled=false, module=ScaleRibbonWidth001}
  ]
  rendererClasses [&NiagaraRibbonRendererProperties]
  renderers [
   {rendererClass=&NiagaraRibbonRendererProperties}
   {rendererClass=&NiagaraRibbonRendererProperties, rendererIndex=1}
   {rendererClass=&NiagaraRibbonRendererProperties, rendererIndex=2}
  ]
 }
}
summary {
 systemName=NS_Boost_Honeybees_Revision
 emitters [
  {emitterName=NE_BeeHive, rendererClasses [&NiagaraSpriteRendererProperties, &NiagaraMeshRendererProperties]}
  {emitterName=NE_HoneyDrops, rendererClasses [&NiagaraSpriteRendererProperties]}
  {emitterName=NE_HoneySwirl, rendererClasses [&NiagaraSpriteRendererProperties, &NiagaraMeshRendererProperties]}
  {emitterName=NE_HoneyTrail001, rendererClasses [&NiagaraRibbonRendererProperties]}
 ]
 userVariables [
  User.bIsFrontendPreview {type=NiagaraBool}
  User.bIsPainted {type=NiagaraBool}
  User.ColorVariantID {type=NiagaraInt32}
  User.DTCarOffset {type=NiagaraPosition, description="", defaultValue {}}
  User.PaintedColor {type=LinearColor, description="", defaultValue {struct=LinearColor, value=<rgba 0.8073 0.2503 0 1>}}
  User.PaintedColor_2 {type=LinearColor, description="", defaultValue {struct=LinearColor, value=<rgba 1 0.3512 0 1>}}
  User.PaintedScalar {type=NiagaraFloat, defaultValue=1, description=""}
  User.PortCount {type=NiagaraInt32, defaultValue=1}
  User.Shine {type=NiagaraFloat, defaultValue=15, description=""}
  User.SingleBoostPos {type=Vector3f, description="", defaultValue {struct=Vector3f, value=<0 0 0>}}
  User.SpecMult {type=NiagaraFloat, defaultValue=1.75, description=""}
  User.SpecPower {type=NiagaraFloat, defaultValue=8, description=""}
  User.SPECULAR_FORCE {type=NiagaraFloat, defaultValue=0, description=""}
 ]
}
