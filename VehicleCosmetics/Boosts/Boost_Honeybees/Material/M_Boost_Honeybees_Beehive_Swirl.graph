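# Unlit, two-sided masked Niagara material for the Honeybees vehicle boost, blending between beehive and honey swirl looks by one scalar.
# The beehive gets stepped, view-dependent shading from a normal map and noise-eroded particle alpha.
# The swirl distorts honey UVs with panning noise in a radial gradient, with dynamic parameters driving distortion, erosion and shine.
# A time-based sine/cosine chain builds an orbit offset vector.
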Material /VehicleCosmetics/Boosts/Boost_Honeybees/Material/M_Boost_Honeybees_Beehive_Swirl
graph {
 properties {
  bAutomaticallySetUsageInEditor=false
  bUsedWithNiagaraMeshParticles=true
  bUsedWithNiagaraRibbons=true
  bUsedWithNiagaraSprites=true
  blendMode=BLEND_Masked
  shadingModel=MSM_Unlit
  twoSided=true
  usedShadingModels=Unlit
 }
 nodes [
  n1 = ParticleColor()
  n2 = TextureSample(UVs=TextureCoordinate() {uTiling=2.5}) {texture=/DealershipSharedLibrary/FX/Textures/Noise/T_VCVFX_Boost_S16_PerlinBlur}
  n3 = Constant() {r=2}
  n4 = TextureCoordinate()
  n5 = Panner(Coordinate=Multiply(A=Subtract(A=@n4) {constB=0.5}, B=ScalarParameter() {defaultValue=1, parameterName=Tile})) {speedY=-1}
  n6 = Multiply(
   A=MaterialFunctionCall() {materialFunction=&RadialGradientExponential}.RadialGradientExponential
   B=TextureSample(UVs=@n5) {texture=&T_VCVFX_Boost_S16_PerlinBlur}.R
  )
  n7 = ComponentMask(None=TextureCoordinate()) {r=true}
  n8 = Constant() {r=0.5}
  n9 = ComponentMask(None=TextureCoordinate()) {g=true}
  n10 = Constant() {r=0.5}
  MaterialExpressionMeshEmitterDynamicParameter_0 = DynamicParameter() {paramNames [Distort_2,Erode,Shines,Param4]}
  n11 = Multiply(A=AppendVector(A=Subtract(A=@n7, B=@n8), B=Subtract(A=@n9, B=@n10)), B=@MaterialExpressionMeshEmitterDynamicParameter_0.Distort_2)
  n12 = TextureSample(UVs=Add(A=TextureCoordinate(), B=Multiply(A=@n6, B=@n11))) {texture=/DealershipSharedLibrary/FX/Textures/TexturePackage/T_VCVF_Boost_Honeybees_HoneySwirl03}
  n13 = TextureSample(UVs=@n5) {texture=/DealershipSharedLibrary/FX/Textures/TexturePackage/T_VCVFX_FoamGradient01_T}
  n14 = ScalarParameter() {parameterName="0=Beehive 1=Swirl"}
  LinearInterpolate(
   A=Saturate(None=Subtract(A=@n1.A, B=@n2.RGB))
   B=Multiply(A=Multiply(A=@n12.R, B=@n3), B=Subtract(A=@n13.G, B=@MaterialExpressionMeshEmitterDynamicParameter_0.Erode))
   Alpha=@n14
  )
  MaterialExpressionMeshEmitterVertexColor_0 = ParticleColor()
  n15 = Multiply(A=@n12.R, B=@MaterialExpressionMeshEmitterVertexColor_0.RGB)
  n16 = Constant3Vector() {constant=<rgba 0.5 0.5 0.1 0>}
  n17 = Multiply(
   A=Desaturation(None=@MaterialExpressionMeshEmitterVertexColor_0.RGB, Fraction=Constant() {r=0.9})
   B=Power(Base=Multiply(A=@n12.G, B=@n16), Exp=DynamicParameter() {paramNames [Distort_2,Erode,Shines,Param4]}.Shines)
  )
  n18 = ParticleColor()
  n19 = Multiply(A=@n18.RGB, B=Constant() {r=0.09})
  n20 = Multiply(
   A=@n18.RGB
   B=Power(
    Base=TextureSample() {texture=/VehicleCosmetics/Boosts/Boost_Honeybees/Textures/T_VCVFX_Boost_HoneyBeeHive_Red}.R
    Exp=ScalarParameter() {defaultValue=1, parameterName=TexturePower}
   )
  )
  n21 = Transform(None=CameraVectorWS()) {transformSourceType=TRANSFORMSOURCE_World, transformType=TRANSFORM_Tangent}
  n22 = TextureSampleParameter2D() {
   parameterName=HoneyBeeHive
   samplerType=SAMPLERTYPE_Normal
   texture=/VehicleCosmetics/Boosts/Boost_Honeybees/Textures/T_VCVFX_Boost_HoneyBeeHive_N
  }
  n23 = Subtract(
   A=Multiply(A=DotProduct(A=@n21, B=@n22.RGB), B=ScalarParameter() {defaultValue=3, parameterName=StepCount})
   B=ScalarParameter() {defaultValue=0.4, parameterName=StepAmount}
  )
  n24 = Saturate(None=ConstantBiasScale(None=Ceil(None=@n23)) {bias=-0.7, scale=1})
  MaterialFunctionCall(Emissive=LinearInterpolate(A=LinearInterpolate(A=@n19, B=@n20, Alpha=@n24), B=Add(A=@n15, B=@n17), Alpha=@n14)) {materialFunction=&MF_FX_EmissiveAdjustments}
  n25 = Multiply(A=Time(), B=ScalarParameter() {defaultValue=3.5, parameterName=OrbitVelocity})
  n26 = Sine(None=@n25)
  n27 = Multiply(A=Cosine(None=@n25), B=ScalarParameter() {defaultValue=0.5, parameterName=OrbitRadiusY})
  n28 = ScalarParameter() {defaultValue=1, parameterName=OrbitRadiusZ}
  Multiply(
   A=AppendVector(A=AppendVector(A=@n26, B=@n27), B=Multiply(A=@n26, B=@n28))
   B=ScalarParameter() {defaultValue=0.25, parameterName=OrbitStrenght}
  )
 ]
}
